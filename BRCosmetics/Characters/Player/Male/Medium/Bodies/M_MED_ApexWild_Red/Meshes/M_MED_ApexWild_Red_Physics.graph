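# Physics asset for the male medium ApexWild Red body, setting up secondary motion for jacket bones around the hips.
# Seven front, side and back chains each link a tiny MSTR root to two capsule segments, with front tips using two capsules.
# Roots and thighs join the pelvis with free rotation, their capsules forming the lower-body volumes.
# Segment joints limit twist and first swing tightly, allowing wider second swing for flap-like bending.
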PhysicsAsset /BRCosmetics/Characters/Player/Male/Medium/Bodies/M_MED_ApexWild_Red/Meshes/M_MED_ApexWild_Red_Physics
bodies {
 dyn_jacket_back_01 [{shapeType=Capsule, length=9.5, radius=2, shapeName=dyn_jacket_side_l_01_capsule, center=<3.4377 0 1.1371>, rotation=<rot 0 0 90>}]
 dyn_jacket_back_02 [{shapeType=Capsule, length=13.5, radius=4.75, shapeName=dyn_jacket_side_l_02_capsule, center=<3.2915 0 4.8327>, rotation=<rot 0 0 90>}]
 dyn_jacket_back_MSTR [{shapeType=Capsule, length=0.505, radius=0.505, shapeName=dyn_jacket_back_MSTR_capsule, rotation=<rot 0 -90 0>}]
 dyn_jacket_back_l_01 [{shapeType=Capsule, length=9.5, radius=2, shapeName=dyn_jacket_side_l_01_capsule, center=<3.4377 0 1.1371>, rotation=<rot 0 0 90>}]
 dyn_jacket_back_l_02 [{shapeType=Capsule, length=13.5, radius=4.75, shapeName=dyn_jacket_side_l_02_capsule, center=<3.2915 0 4.8327>, rotation=<rot 0 0 90>}]
 dyn_jacket_back_l_MSTR [{shapeType=Capsule, length=0.505, radius=0.505, shapeName=dyn_jacket_side_l_MSTR_capsule, rotation=<rot 0 -90 0>}]
 dyn_jacket_back_r_01 [{shapeType=Capsule, length=9.5, radius=2, shapeName=dyn_jacket_side_l_01_capsule, center=<3.4377 0 1.1371>, rotation=<rot 0 0 90>}]
 dyn_jacket_back_r_02 [{shapeType=Capsule, length=13.5, radius=4.75, shapeName=dyn_jacket_side_l_02_capsule, center=<3.2915 0 4.8327>, rotation=<rot 0 0 90>}]
 dyn_jacket_back_r_MSTR [{shapeType=Capsule, length=0.505, radius=0.505, shapeName=dyn_jacket_side_l_MSTR_capsule, rotation=<rot 0 -90 0>}]
 dyn_jacket_fr_l_01 [{shapeType=Capsule, length=7.75, radius=2, shapeName=dyn_jacket_side_l_01_capsule, center=<4.7 0.0588 1.7528>, rotation=<rot 0 0 95.4564>}]
 dyn_jacket_fr_l_02 [
  {
   shapeType=Capsule
   length=10.5
   radius=5
   shapeName=dyn_jacket_side_l_02_capsule
   center=<4.7697 0.5802 5>
   rotation=<rot 0.0199 0.0078 97.3778>
  }
  {
   shapeType=Capsule
   length=9
   radius=5
   shapeName=dyn_jacket_fr_l_02_capsule
   center=<4.9956 -11.6699 3.8363>
   rotation=<rot 0.0001 0 59.3909>
  }
 ]
 dyn_jacket_fr_l_MSTR [{shapeType=Capsule, length=0.505, radius=0.505, shapeName=dyn_jacket_fr_l_MSTR_capsule, rotation=<rot 0 -90 0>}]
 dyn_jacket_fr_r_01 [{
   shapeType=Capsule
   length=7.75
   radius=2
   shapeName=dyn_jacket_side_l_01_capsule
   center=<4.7 0.0588 1.7528>
   rotation=<rot 0.0001 0 84.5167>
  }]
 dyn_jacket_fr_r_02 [
  {shapeType=Capsule, length=10.5, radius=5, shapeName=dyn_jacket_side_l_02_capsule, center=<4.7697 -0.5802 5>, rotation=<rot 0.0001 0 80.7441>}
  {shapeType=Capsule, length=9, radius=5, shapeName=dyn_jacket_fr_r_02_capsule, center=<4.9956 11.3046 3.8363>, rotation=<rot 0 0 -59.3909>}
 ]
 dyn_jacket_fr_r_MSTR [{shapeType=Capsule, length=0.505, radius=0.505, shapeName=dyn_jacket_fr_r_MSTR_capsule, rotation=<rot 0 -90 0>}]
 dyn_jacket_side_l_01 [{shapeType=Capsule, length=9.5, radius=2, shapeName=dyn_jacket_side_l_01_capsule, center=<3.4377 0 1.1371>, rotation=<rot 0 0 90>}]
 dyn_jacket_side_l_02 [{shapeType=Capsule, length=9.25, radius=3.25, shapeName=dyn_jacket_side_l_02_capsule, center=<4.1136 0 5.1396>, rotation=<rot 0 0 90>}]
 dyn_jacket_side_l_MSTR [{shapeType=Capsule, length=0.505, radius=0.505, shapeName=dyn_jacket_side_l_MSTR_capsule, rotation=<rot 0 -90 0>}]
 dyn_jacket_side_r_01 [{shapeType=Capsule, length=9.5, radius=2, shapeName=dyn_jacket_side_l_01_capsule, center=<3.4377 0 1.1371>, rotation=<rot 0 -0.0001 80>}]
 dyn_jacket_side_r_02 [{
   shapeType=Capsule
   length=9.25
   radius=3.25
   shapeName=dyn_jacket_side_l_02_capsule
   center=<4.1136 -1.9066 4.5503>
   rotation=<rot 0 -0.0002 76.2519>
  }]
 dyn_jacket_side_r_MSTR [{shapeType=Capsule, length=0.505, radius=0.505, shapeName=dyn_jacket_side_r_MSTR_capsule, rotation=<rot 0 -90 0>}]
 pelvis [
  {shapeType=Capsule, length=5.75, radius=10.4661, center=<3.545 -0.7185 -0.0436>}
  {shapeType=Capsule, length=11.5, radius=5, shapeName=pelvis_capsule, center=<-2.4602 -5.7666 0>}
 ]
 thigh_l [{
   shapeType=Capsule
   length=15.282
   radius=8.6749
   shapeName=thigh_l_capsule
   center=<-13.0247 1.1254 0.4443>
   rotation=<rot 90 89.2957 90.1915>
  }]
 thigh_r [{
   shapeType=Capsule
   length=15.282
   radius=8.6749
   shapeName=thigh_l_capsule
   center=<13.0247 -1.1254 -0.4443>
   rotation=<rot 90 89.2957 87.2311>
  }]
}
constraints [
 {
  bone1Name=thigh_l
  bone2Name=pelvis
  swing1LimitDegrees=45
  swing1Motion=Free
  swing2LimitDegrees=45
  swing2Motion=Free
  twistLimitDegrees=45
  twistMotion=Free
 }
 {
  bone1Name=thigh_r
  bone2Name=pelvis
  swing1LimitDegrees=45
  swing1Motion=Free
  swing2LimitDegrees=45
  swing2Motion=Free
  twistLimitDegrees=45
  twistMotion=Free
 }
 {bone1Name=dyn_jacket_side_l_02, bone2Name=dyn_jacket_side_l_01, swing1LimitDegrees=5, swing2LimitDegrees=45, twistLimitDegrees=10}
 {bone1Name=dyn_jacket_side_l_01, bone2Name=dyn_jacket_side_l_MSTR, swing1LimitDegrees=5, swing2LimitDegrees=25, twistLimitDegrees=2}
 {bone1Name=dyn_jacket_fr_l_02, bone2Name=dyn_jacket_fr_l_01, swing1LimitDegrees=5, swing2LimitDegrees=45, twistLimitDegrees=10}
 {bone1Name=dyn_jacket_fr_l_01, bone2Name=dyn_jacket_fr_l_MSTR, swing1LimitDegrees=15, swing2LimitDegrees=45, twistLimitDegrees=2}
 {
  bone1Name=dyn_jacket_side_l_MSTR
  bone2Name=pelvis
  swing1LimitDegrees=45
  swing1Motion=Free
  swing2LimitDegrees=45
  swing2Motion=Free
  twistLimitDegrees=45
  twistMotion=Free
 }
 {
  bone1Name=dyn_jacket_fr_l_MSTR
  bone2Name=pelvis
  swing1LimitDegrees=45
  swing1Motion=Free
  swing2LimitDegrees=45
  swing2Motion=Free
  twistLimitDegrees=45
  twistMotion=Free
 }
 {bone1Name=dyn_jacket_back_l_02, bone2Name=dyn_jacket_back_l_01, swing1LimitDegrees=2, swing2LimitDegrees=45, twistLimitDegrees=10}
 {bone1Name=dyn_jacket_back_l_01, bone2Name=dyn_jacket_back_l_MSTR, swing1LimitDegrees=5, swing2LimitDegrees=35, twistLimitDegrees=2}
 {
  bone1Name=dyn_jacket_back_l_MSTR
  bone2Name=pelvis
  swing1LimitDegrees=45
  swing1Motion=Free
  swing2LimitDegrees=45
  swing2Motion=Free
  twistLimitDegrees=45
  twistMotion=Free
 }
 {bone1Name=dyn_jacket_back_01, bone2Name=dyn_jacket_back_MSTR, swing1LimitDegrees=5, swing2LimitDegrees=35, twistLimitDegrees=2}
 {
  bone1Name=dyn_jacket_back_MSTR
  bone2Name=pelvis
  swing1LimitDegrees=45
  swing1Motion=Free
  swing2LimitDegrees=45
  swing2Motion=Free
  twistLimitDegrees=45
  twistMotion=Free
 }
 {bone1Name=dyn_jacket_back_02, bone2Name=dyn_jacket_back_01, swing1LimitDegrees=2, swing2LimitDegrees=45, twistLimitDegrees=10}
 {
  bone1Name=dyn_jacket_back_r_MSTR
  bone2Name=pelvis
  swing1LimitDegrees=45
  swing1Motion=Free
  swing2LimitDegrees=45
  swing2Motion=Free
  twistLimitDegrees=45
  twistMotion=Free
 }
 {bone1Name=dyn_jacket_back_r_01, bone2Name=dyn_jacket_back_r_MSTR, swing1LimitDegrees=5, swing2LimitDegrees=35, twistLimitDegrees=2}
 {bone1Name=dyn_jacket_back_r_02, bone2Name=dyn_jacket_back_r_01, swing1LimitDegrees=2, swing2LimitDegrees=45, twistLimitDegrees=10}
 {
  bone1Name=dyn_jacket_fr_r_MSTR
  bone2Name=pelvis
  swing1LimitDegrees=45
  swing1Motion=Free
  swing2LimitDegrees=45
  swing2Motion=Free
  twistLimitDegrees=45
  twistMotion=Free
 }
 {bone1Name=dyn_jacket_fr_r_01, bone2Name=dyn_jacket_fr_r_MSTR, swing1LimitDegrees=15, swing2LimitDegrees=45, twistLimitDegrees=2}
 {bone1Name=dyn_jacket_fr_r_02, bone2Name=dyn_jacket_fr_r_01, swing1LimitDegrees=5, swing2LimitDegrees=45, twistLimitDegrees=10}
 {
  bone1Name=dyn_jacket_side_r_MSTR
  bone2Name=pelvis
  swing1LimitDegrees=45
  swing1Motion=Free
  swing2LimitDegrees=45
  swing2Motion=Free
  twistLimitDegrees=45
  twistMotion=Free
 }
 {bone1Name=dyn_jacket_side_r_01, bone2Name=dyn_jacket_side_r_MSTR, swing1LimitDegrees=5, swing2LimitDegrees=25, twistLimitDegrees=2}
 {bone1Name=dyn_jacket_side_r_02, bone2Name=dyn_jacket_side_r_01, swing1LimitDegrees=2, swing2LimitDegrees=55, twistLimitDegrees=10}
]
