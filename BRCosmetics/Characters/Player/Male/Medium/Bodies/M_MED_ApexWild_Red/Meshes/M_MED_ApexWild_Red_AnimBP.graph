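# Post-process style character-part Anim Blueprint for the ApexWild Red male body, copying the attached parent mesh's pose and adding secondary physics motion.
# The copied pose runs through two Control Rigs and additive collar-correction bone offsets, then is cached.
# The cache feeds two pelvis-space RigidBody branches with different physics assets, joined by a bool blend; the FE branch is softer.
# Each update recomputes the gravity override used as both simulations' external force.
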AnimBlueprint /BRCosmetics/Characters/Player/Male/Medium/Bodies/M_MED_ApexWild_Red/Meshes/M_MED_ApexWild_Red_AnimBP
target_skeleton=&/M_MED_ApexWild_Red_Skeleton
graphs [
 AnimGraph {nodes [
   n1 = Get GravityOverride()
   n2 = RigidBody(ComponentPose=Local To Component(LocalPose=Use cached pose 'PriorRban'()), Alpha=0.7, ExternalForce=@n1) {
    overridePhysicsAsset=&/M_MED_ApexWild_Red_Physics_FE
    simulationSpace=BaseBoneSpace
    alphaBoolBlend {blendOption=Linear}
    baseBoneRef {boneName=pelvis}
    componentAppliedLinearAccClamp=<30000 30000 30000>
    componentLinearAccScale=<0.2 0.2 0.3>
    componentLinearVelScale=<0.2 0.2 0.3>
   }
   n3 = RigidBody(ComponentPose=Local To Component(LocalPose=Use cached pose 'PriorRban'()), ExternalForce=@n1) {
    overridePhysicsAsset=&/M_MED_ApexWild_Red_Physics
    simulationSpace=BaseBoneSpace
    alphaBoolBlend {blendOption=Linear}
    baseBoneRef {boneName=pelvis}
    componentAppliedLinearAccClamp=<30000 30000 30000>
    componentLinearAccScale=<0.3 0.3 0.5>
    componentLinearVelScale=<0.3 0.3 0.5>
   }
   Output Pose(Result=Blend Poses by bool(BlendPose_0=Component To Local(ComponentPose=@n2), BlendPose_1=Component To Local(ComponentPose=@n3)) {blendType=Linear}) {layerGroup=None}
   n4 = Control Rig(Source=Copy Pose From Mesh() {bUseAttachedParent=true}) {
    alphaBoolBlend {blendOption=Linear}
    controlRigAssetReference {blueprintRigClass=&/M_MED_ApexWild_Red_CtrlRig.M_MED_ApexWild_Red_CtrlRig_C}
   }
   n5 = Control Rig(Source=@n4) {
    alphaBoolBlend {blendOption=Linear}
    controlRigAssetReference {blueprintRigClass=&/M_ApexWildRed_DeformRig.M_ApexWildRed_DeformRig_C}
   }
   n6 = "Transform (Modify) Bone - Bone: dyn_collarcorrection_L"(ComponentPose=Local To Component(LocalPose=@n5), Alpha=1, Translation="1.140,1.455,-0.000", Rotation="0.000,39.993,-0.000", Scale="1,1,1") {
    rotationMode=BMM_Additive
    translationMode=BMM_Additive
    alphaBoolBlend {blendOption=Linear}
    boneToModify {boneName=dyn_collarcorrection_L}
    rotation {yaw=39.993}
    translation=<1.14 1.455>
   }
   n7 = "Transform (Modify) Bone - Bone: dyn_collarcorrection_R"(ComponentPose=@n6, Alpha=1, Translation="-1.153,-0.151,0.000", Rotation="0.000,17.181,-0.000", Scale="1,1,1") {
    rotationMode=BMM_Additive
    translationMode=BMM_Additive
    alphaBoolBlend {blendOption=Linear}
    boneToModify {boneName=dyn_collarcorrection_R}
    rotation {yaw=17.181}
    translation=<-1.153 -0.151>
   }
   Save cached pose 'PriorRban'(Pose=Component To Local(ComponentPose=@n7)) {cacheName=PriorRban}
  ]}
 EventGraph {nodes [
   Event BlueprintUpdateAnimation()
   > Sequence()
   > n1 = Gravity Override State Function DEPRECATED(CharacterPartAnimInstance=Self-Reference(), GameTimeInSeconds=GetGameTimeInSeconds(), GravityOverrideParams=Get Gravity Override Params())
   > Set GravityOverride(GravityOverride=@n1)
  ]}
]
variables [
 Gravity Override Params {type=GravityOverrideParamsStruct}
 GravityOverride {type=Vector}
]
