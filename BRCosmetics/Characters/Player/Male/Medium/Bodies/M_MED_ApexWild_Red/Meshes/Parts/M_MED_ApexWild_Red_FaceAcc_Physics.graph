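# Physics asset for the face accessory mesh of the M_MED_ApexWild_Red male character body, setting up physics for dynamic hair and strand bones.
# A capsule on the head anchors six hair capsules and the left, middle and right strand roots through angle-limited constraints.
# The left and middle strands chain to a second segment, while the right strand has one.
# Twist limits are tight throughout, with wider swing that varies per strand.
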PhysicsAsset /BRCosmetics/Characters/Player/Male/Medium/Bodies/M_MED_ApexWild_Red/Meshes/Parts/M_MED_ApexWild_Red_FaceAcc_Physics
bodies {
 dyn_hair_01 [{shapeType=Capsule, length=4, radius=4, shapeName=dyn_hair_01_capsule, center=<0 0 -1>, rotation=<rot 0 90 0>}]
 dyn_hair_02 [{shapeType=Capsule, length=4, radius=4, shapeName=dyn_hair_02_capsule, center=<0 0 -1>, rotation=<rot 0 90 0>}]
 dyn_hair_03 [{shapeType=Capsule, length=4, radius=4, shapeName=dyn_hair_03_capsule, center=<0 0 -1>, rotation=<rot 0 90 0>}]
 dyn_hair_04 [{shapeType=Capsule, length=4, radius=4, shapeName=dyn_hair_04_capsule, center=<0 0 -1>, rotation=<rot 0 90 0>}]
 dyn_hair_05 [{shapeType=Capsule, length=4, radius=4, shapeName=dyn_hair_05_capsule, center=<0 0 -1>, rotation=<rot 0 90 0>}]
 dyn_hair_06 [{shapeType=Capsule, length=4, radius=4, shapeName=dyn_hair_06_capsule, center=<0 0 -1>, rotation=<rot 0 90 0>}]
 dyn_strand_01_l [{shapeType=Capsule, length=0.75, radius=0.75, shapeName=dyn_strand_01_l_capsule, center=<0.5176 0 0>, rotation=<rot 90 0 0>}]
 dyn_strand_01_m [{shapeType=Capsule, length=0.505, radius=0.7, shapeName=dyn_strand_01_m_capsule, center=<0.4574 0 0>, rotation=<rot 90 0 0>}]
 dyn_strand_01_r [{shapeType=Capsule, length=0.505, radius=0.505, shapeName=dyn_strand_01_r_capsule, center=<0.6344 0 0>, rotation=<rot 90 0 0>}]
 dyn_strand_02_l [{shapeType=Capsule, length=1.3736, radius=0.5, shapeName=dyn_strand_02_l_capsule, center=<0.7177 0 0>, rotation=<rot 90 0 0>}]
 dyn_strand_02_m [{shapeType=Capsule, length=0.7435, radius=0.5544, shapeName=dyn_strand_02_m_capsule, center=<0.5119 0 0>, rotation=<rot 90 0 0>}]
 head [{shapeType=Capsule, length=0.75, radius=7.25, center=<4.5831 2.6006 -0.163>, rotation=<rot 0.6815 0.2006 -87.2991>}]
}
constraints [
 {bone1Name=dyn_strand_01_l, bone2Name=head, swing1LimitDegrees=39, swing2LimitDegrees=39, twistLimitDegrees=3}
 {bone1Name=dyn_strand_02_l, bone2Name=dyn_strand_01_l, swing1LimitDegrees=35, swing2LimitDegrees=35, twistLimitDegrees=3}
 {bone1Name=dyn_strand_01_m, bone2Name=head, swing1LimitDegrees=10, swing2LimitDegrees=10, twistLimitDegrees=3}
 {bone1Name=dyn_strand_02_m, bone2Name=dyn_strand_01_m, swing1LimitDegrees=15, swing2LimitDegrees=15, twistLimitDegrees=3}
 {bone1Name=dyn_strand_01_r, bone2Name=head, swing1LimitDegrees=15, swing2LimitDegrees=15, twistLimitDegrees=3}
 {bone1Name=dyn_hair_01, bone2Name=head, swing1LimitDegrees=20, swing2LimitDegrees=20, twistLimitDegrees=5}
 {bone1Name=dyn_hair_02, bone2Name=head, swing1LimitDegrees=20, swing2LimitDegrees=20, twistLimitDegrees=5}
 {bone1Name=dyn_hair_03, bone2Name=head, swing1LimitDegrees=20, swing2LimitDegrees=20, twistLimitDegrees=5}
 {bone1Name=dyn_hair_04, bone2Name=head, swing1LimitDegrees=20, swing2LimitDegrees=20, twistLimitDegrees=5}
 {bone1Name=dyn_hair_05, bone2Name=head, swing1LimitDegrees=20, swing2LimitDegrees=20, twistLimitDegrees=5}
 {bone1Name=dyn_hair_06, bone2Name=head, swing1LimitDegrees=20, swing2LimitDegrees=20, twistLimitDegrees=5}
]
